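# Niagara system for the Shooting Star skydiving contrail cosmetic, with all four emitters attached to bones of the user skeletal mesh, selected by spawn group.
# Ribbon draws fading, tapering trails keyed by bone, while Hand_Glow sprites resample bone positions every update to stay attached.
# Sparkles spawn by distance moved through User.Spawn Per Unit and stretch into fast streaks.
# Comet_Mesh burst-spawns a short-lived, rotated mesh that fades out.
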NiagaraSystem /BRCosmetics/Athena/Effects/Particles/Skydiving/NS_Contrail_ShootingStar/NS_Contrail_ShootingStar
dependencies {
 usedDataInterfaces [&NiagaraDataInterfaceCurve, &NiagaraDataInterfaceVector2DCurve]
 usedDynamicInputs []
 usedModules [/Niagara/Modules/System/SystemLifeCycle]
}
emitters {
 Comet_Mesh {
  emitterUpdateScript [
   &EmitterLifeCycle(NextLoopDuration=0.1)
   &SpawnBurst_Instantaneous(Spawn Count=1, Use Spawn Probability=false)
  ]
  particleSpawnScript [
   /Niagara/Modules/Spawn/Initialization/InitializeParticle(
    Lifetime=&UniformRangedFloat(Minimum=0.2, Maximum=0.2)
    Mesh Scale=&VectorFromFloat(Value=&UniformRangedFloat(Minimum=1.5, Maximum=1.5))
    Module.Write Color=true
    Module.Write Lifetime=true
    Module.Write Mass=false
    Module.Write Material Random=true
    Module.Write Position=false
    Module.Write Scale=true
    Module.Write SpriteRotation=false
    Module.Write SpriteSize=true
    Sprite Size=<3500 3500>
   )
   &SampleSkeletalMeshSkeleton(Mesh:NiagaraDataInterfaceSkeletalMesh=$User.SkeletonMesh, Specific Bone Index:NiagaraInt32=&SpawnGroup())
   /Niagara/Modules/Spawn/Location/SkeletalMeshSkeletonLocation(Module.Inherit Velocity=false, Module.Use Sampled Bone as Ribbon ID=true)
   &SkeletalMeshSkeletonLocation(Module.Inherit Velocity=false, Module.Use Sampled Bone as Ribbon ID=true, Sampled Position Offset:Vector3f=<-40 0 0>) {module=SkeletalMeshSkeletonLocation001}
   /Niagara/Modules/Spawn/Orientation/InitialMeshRotation(Module.Use Orientation Axis=false, Pitch=-33, Yaw=90)
  ]
  particleUpdateScript [
   &ParticleState() {module=UpdateAge}
   &Color(Scale Alpha=&FloatFromCurve(FloatCurve=DataInterface() {propertyValues {Curve {keys [<at -0.0046 1.0113>,<at 1.0011 0.002>]}}}))
   &SolveForcesAndVelocity(Clamp Velocity=false, Limit Acceleration=false)
  ]
  rendererClasses [&NiagaraMeshRendererProperties]
  renderers [{rendererClass=&NiagaraMeshRendererProperties, rendererIndex=0}]
 }
 Hand_Glow {
  emitterUpdateScript [
   &EmitterLifeCycle(NextLoopDelay=-0.0625, NextLoopDuration=0.5)
   /Game/Effects/Niagara/Modules/Spawn/SpawnRateFromMultipleSpawnGroups_4()
   &SpawnRateFromMultipleSpawnGroups_4(Global Spawnrate=32) {module=SpawnRateFromMultipleSpawnGroups_4001}
  ]
  particleSpawnScript [&InitializeParticle(
    Lifetime=&UniformRangedFloat(Minimum=0.1, Maximum=0.15)
    Module.Write Color=true
    Module.Write Lifetime=true
    Module.Write Mass=false
    Module.Write Material Random=true
    Module.Write Position=false
    Module.Write Scale=false
    Module.Write SpriteRotation=false
    Module.Write SpriteSize=true
    Sprite Size=&Vector2DFromFloat(Value=&UniformRangedFloat(Minimum=60, Maximum=65))
   )]
  particleUpdateScript [
   &SampleSkeletalMeshSkeleton(Mesh:NiagaraDataInterfaceSkeletalMesh=$User.SkeletonMesh, Specific Bone Index:NiagaraInt32=&SpawnGroup())
   &SkeletalMeshSkeletonLocation(Module.Inherit Velocity=false, Module.Use Sampled Bone as Ribbon ID=true)
   &ParticleState() {module=UpdateAge}
   &Color(
    Color=<rgba 0 8.5303 50 1>
    Scale Alpha=&FloatFromCurve(FloatCurve=DataInterface() {propertyValues {Curve {keys [<at -0.0046 1.0113>,<at 1.0011 0.002>]}}})
   )
   &CameraOffset(Camera Offset Amount=&UniformRangedFloat(Minimum=-15, Maximum=-15))
   &ScaleSpriteSize(Scale Factor=&Vector2DFromFloat(Value=&FloatFromCurve(FloatCurve=DataInterface() {propertyValues {Curve {keys [{interpMode=RCIM_Cubic, tangentMode=RCTM_User, time=0, value=0.6}, {interpMode=RCIM_Cubic, leaveTangent=-0.039, time=1.0011, value=1}]}}}))) {module=SpriteSizeScale}
   &SolveForcesAndVelocity(Clamp Velocity=false, Limit Acceleration=false)
  ]
  rendererClasses [&NiagaraSpriteRendererProperties]
  renderers [{rendererClass=&NiagaraSpriteRendererProperties}]
 }
 Ribbon {
  emitterUpdateScript [
   &EmitterLifeCycle(NextLoopDelay=-0.0625)
   /SetVariables() {isSetParameters=true}
   &SpawnRateFromMultipleSpawnGroups_4(
    Global Spawnrate=40
    SpawnGroup0:NiagaraInt32=&UniformRangedInt(Minimum=0, Maximum=0)
    SpawnGroup1:NiagaraInt32=&UniformRangedInt(Minimum=1, Maximum=1)
    SpawnGroup2:NiagaraInt32=&UniformRangedInt(Minimum=2, Maximum=2)
    SpawnGroup3:NiagaraInt32=&UniformRangedInt(Minimum=3, Maximum=3)
   )
  ]
  particleSpawnScript [
   /SetVariables(Particles.SpawnRate:NiagaraFloat=$Emitter.SpawnRate) {isSetParameters=true}
   &InitializeRibbon(
    Color=<rgba 0.02 0.36 1 0.25>
    Lifetime=0.3
    Module.Write Color=true
    Module.Write Lifetime=true
    Module.Write Mass=true
    Module.Write Material Random=true
    Module.Write Position=false
    Module.Write RibbonTwist=false
    Module.Write RibbonWidth=true
    Position:Vector3f=&SimulationPosition()
    Ribbon Twist=1
    Ribbon Width=32
   )
   &SampleSkeletalMeshSkeleton(Mesh:NiagaraDataInterfaceSkeletalMesh=$User.SkeletonMesh, Specific Bone Index:NiagaraInt32=&SpawnGroup())
   &SkeletalMeshSkeletonLocation(Inherit Velocity From Mesh=5, Module.Inherit Velocity=false, Module.Use Sampled Bone as Ribbon ID=true)
  ]
  particleUpdateScript [
   &ParticleState() {module=UpdateAge}
   &Color(
    Color=<rgba 1 1 1 0.25>
    Scale Alpha=&FloatFromCurve(FloatCurve=DataInterface() {propertyValues {Curve {keys [<at -0.0046 1.0113>,<at 1.0011 0.002>]}}})
   )
   &AddVelocity(Velocity=<0 -500 500>)
   &SolveForcesAndVelocity(Clamp Velocity=false, Limit Acceleration=false)
   /SetVariables(Particles.RibbonWidth=&FloatFromCurve(FloatCurve=DataInterface() {propertyValues {Curve {keys [
         <at 0 35.8443 RCIM_Cubic>
         {arriveTangent=62.1805, interpMode=RCIM_Cubic, leaveTangent=62.1792, tangentMode=RCTM_User, time=0.1007, value=45.0317}
         {arriveTangent=2.1843, interpMode=RCIM_Cubic, leaveTangent=2.1843, tangentMode=RCTM_User, time=1, value=0.7054}
        ]}}})) {isSetParameters=true}
   /Niagara/Modules/Update/Material/DynamicMatParams(
    Module.Param0WriteEnabled=true
    Module.Param1WriteEnabled=true
    Module.Param2WriteEnabled=false
    Module.Param3WriteEnabled=false
    Param0Value=0.25
    Param1Value=0.25
   )
  ]
  rendererClasses [&NiagaraRibbonRendererProperties]
  renderers [{rendererClass=&NiagaraRibbonRendererProperties}]
 }
 Sparkles {
  emitterUpdateScript [
   &EmitterLifeCycle(NextLoopDelay=-0.0625, NextLoopDuration=0.5)
   &SpawnRateFromMultipleSpawnGroups_4(Global Spawnrate=0)
   &SpawnPerUnit(
    Max Movement Threshold=12000
    Module.Use Max Movement Threshold=true
    Module.Use Movement Tolerance=false
    Spawn Per Unit:NiagaraFloat=$User.Spawn Per Unit
    Use Spawn Probability=false
   )
   &SpawnPerUnit(
    Max Movement Threshold=12000
    Module.Use Max Movement Threshold=true
    Module.Use Movement Tolerance=false
    Spawn Group=1
    Spawn Per Unit:NiagaraFloat=$User.Spawn Per Unit
    Use Spawn Probability=false
   ) {module=SpawnPerUnit001}
   &SpawnPerUnit(
    Max Movement Threshold=12000
    Module.Use Max Movement Threshold=true
    Module.Use Movement Tolerance=false
    Spawn Group=2
    Spawn Per Unit:NiagaraFloat=$User.Spawn Per Unit
    Use Spawn Probability=false
   ) {module=SpawnPerUnit002}
   &SpawnPerUnit(
    Max Movement Threshold=12000
    Module.Use Max Movement Threshold=true
    Module.Use Movement Tolerance=false
    Spawn Group=3
    Spawn Per Unit:NiagaraFloat=$User.Spawn Per Unit
    Use Spawn Probability=false
   ) {module=SpawnPerUnit003}
  ]
  particleSpawnScript [
   &InitializeParticle(
    Color=<rgba 0 35 100 1>
    Lifetime=&UniformRangedFloat(Minimum=0.15, Maximum=0.3)
    Module.Write Color=true
    Module.Write Lifetime=true
    Module.Write Mass=false
    Module.Write Material Random=true
    Module.Write Position=false
    Module.Write Scale=false
    Module.Write SpriteRotation=false
    Module.Write SpriteSize=true
    Sprite Size=&Vector2DFromFloat(Value=&UniformRangedFloat(Minimum=2, Maximum=3))
   )
   &SampleSkeletalMeshSkeleton(Mesh:NiagaraDataInterfaceSkeletalMesh=$User.SkeletonMesh, Specific Bone Index:NiagaraInt32=&SpawnGroup())
   &SkeletalMeshSkeletonLocation(Module.Inherit Velocity=true, Module.Use Sampled Bone as Ribbon ID=true)
   /Niagara/Modules/Spawn/Location/SphereLocation(MaskSpawnGroup=false, Module.Override Randomness=false, Sphere Radius=12, Surface Only=false)
   &AddVelocity(Scale Added Velocity=<0.1 0.1 1>, Velocity=<0 -7000 750>)
  ]
  particleUpdateScript [
   &ParticleState() {module=UpdateAge}
   &Color(
    Color=<rgba 0 24.4838 100 1>
    Scale Alpha=&FloatFromCurve(FloatCurve=DataInterface() {propertyValues {Curve {keys [<at -0.0046 1.0113>,<at 1.0011 0.002>]}}})
   )
   &CameraOffset(Camera Offset Amount=&UniformRangedFloat(Minimum=0, Maximum=1))
   &SolveForcesAndVelocity(Clamp Velocity=false, Limit Acceleration=false)
   &ScaleSpriteSize(Scale Factor=&Vector2DFromCurve(
     Vector2Curve=DataInterface() {propertyValues {XCurve {keys [<at 0 1>,<at 0.4987 0.5007>,<at 0.9989 0>]}, YCurve {keys [<at 0 0 RCIM_Cubic>,<at 0.5 40 RCIM_Cubic>,<at 0.9989 0>]}}}
     ScaleCurve=<1 1>
    )) {module=SpriteSizeScale}
  ]
  rendererClasses [&NiagaraSpriteRendererProperties]
  renderers [{rendererClass=&NiagaraSpriteRendererProperties}]
 }
}
summary {
 systemName=NS_Contrail_ShootingStar
 emitters [
  {emitterName=Ribbon, rendererClasses [&NiagaraRibbonRendererProperties]}
  {emitterName=Sparkles, rendererClasses [&NiagaraSpriteRendererProperties]}
  {emitterName=Comet_Mesh, rendererClasses [&NiagaraMeshRendererProperties]}
  {emitterName=Hand_Glow, rendererClasses [&NiagaraSpriteRendererProperties]}
 ]
 userVariables [
  User.SkeletonMesh {
   type=NiagaraDataInterfaceSkeletalMesh
   defaultValue {
    struct=NiagaraExt_VariableValue_DataInterface
    value {dataInterface=&self:NiagaraDataInterfaceSkeletalMesh_0}
   }
  }
  User.Spawn Per Unit {type=NiagaraFloat, defaultValue=250, description=""}
 ]
}
